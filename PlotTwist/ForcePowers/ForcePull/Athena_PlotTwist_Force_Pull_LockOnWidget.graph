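# Force Pull lock-on widget that shows a centered, pulsing cyan reticule image using a lock-on cursor material.
# Each Tick takes the absolute sine of server world time scaled by PulseInterval, so higher values pulse faster despite the name.
# That wave blends between SizeMin and SizeMax, and UpdateSizeAndIntensity writes the result to both CenterSize and Intensity on the image's dynamic material.
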WidgetBlueprint /PlotTwist/ForcePowers/ForcePull/Athena_PlotTwist_Force_Pull_LockOnWidget
graphs [EventGraph {nodes [
   n1 = UpdateSizeAndIntensity!() -> (OutputDelegate, then, SizeAndIntensity:double=~) {is=title K2Node_CustomEvent}
   > n2 = GetDynamicMaterial(self=Get Image_LockOnReticule() -> (Image_LockOnReticule:Image Object Reference))
   > SetScalarParameterValue(self=@n2, ParameterName=CenterSize, Value=@n1.SizeAndIntensity)
   > SetScalarParameterValue(self=@n2, ParameterName=Intensity, Value=@n1.SizeAndIntensity)
   Event Tick()
   n3 = "Sin (Degrees)"!(A:double=float * float!(A=GetServerWorldTimeSeconds!() -> (ReturnValue:float), B=180, C:double=Get PulseInterval() -> (PulseInterval:double)) -> (ReturnValue) {is=title K2Node_PromotableOperator}) -> (ReturnValue:double)
   > UpdateSizeAndIntensity!(
    execute
    self
    SizeAndIntensity:double=Lerp(A=Get SizeMin() -> (SizeMin:double), B=Get SizeMax() -> (SizeMax:double), Alpha="Absolute (Float)"(A=@n3))
   ) -> (then) {is=title K2Node_CallFunction}
  ]}]
variables [
 PulseInterval {type=number, default=1}
 SizeMax {type=number, default=1}
 SizeMin {type=number, default=0.8}
]
widgets {
 info {inheritedWidgetCount=0, parentClass=&UserWidget, rootWidgetClass=&CanvasPanel, widgetCount=2}
 widgets [
  CanvasPanel_29 {widgetClassPath=&CanvasPanel, widget=CanvasPanel_29}
  Image_LockOnReticule {
   widgetClassPath=&Image
   bIsVariable=true
   parent=CanvasPanel_29
   widget=Image_LockOnReticule
   properties {
    renderOpacity=2
    brush {
     bIsDynamicallyLoaded=false
     drawAs=Image
     imageType=NoImage
     mirroring=NoMirror
     resourceName=None
     resourceObject=/Game/Athena/Items/Gameplay/BackPacks/Ashton/Indigo/FX/M_Ashton_Indigo_LockOnCursor
     tiling=NoTile
     imageSize=<32 32>
     margin=<ltrb 0 0 0 0>
     outlineSettings {
      bUseBrushTransparency=false
      roundingType=HalfHeightRadius
      width=0
      color {colorUseRule=UseColor_Specified, specifiedColor=<rgba 0 0 0 0>}
      cornerRadii=<0 0 0 0>
     }
     tintColor {colorUseRule=UseColor_Specified, specifiedColor=<rgba 0 0.9894 1 1>}
     uVRegion {bIsValid=false, max=<0 0>, min=<0 0>}
    }
    colorAndOpacity=<rgba 1 1 1 0.9>
   }
   slot {
    class=CanvasPanelSlot
    properties {layoutData {alignment=<0.5 0.5>, anchors {maximum=<0.5 0.5>, minimum=<0.5 0.5>}, offsets=<ltrb 0 0 250 250>}}
   }
  }
 ]
}
