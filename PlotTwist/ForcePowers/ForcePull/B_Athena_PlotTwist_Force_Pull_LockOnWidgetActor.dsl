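; Lock-on indicator actor for the PlotTwist Force Pull power, showing a widget on the targeted actor.
; The widget starts hidden on BeginPlay.
; NewTarget stores a valid target, makes the widget HitTestInvisible, moves the indicator using the target's collision bounds and attaches it keeping world transform.
; An invalid target calls TargetLost, which detaches the indicator and hides the widget.
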(event EventBeginPlay
  (Widget|SetVisibility (UserInterface|GetUserWidgetObject (Variables|Default|GetWidget)) "Hidden"))

(fn NewTarget
  (params (Lock%20On%20Target :type Actor%20Object%20Reference))
  (Utilities|IsValid Lock%20On%20Target
    (:Is%20Valid
      (Variables|Default|SetLockOnTargetActor (Variables|GetLock%20On%20Target))
      (Widget|SetVisibility (UserInterface|GetUserWidgetObject (Variables|Default|GetWidget)) "HitTestInvisible")
      (Transformation|SetActorLocation (Fortnite|GetActorCollisionComponentBounds (Variables|GetLock%20On%20Target)) false true)
      (Transformation|AttachActorToActor (Variables|GetLock%20On%20Target) "None" "KeepWorld" "KeepWorld" "KeepWorld"))
    (:Is%20Not%20Valid
      (CallFunction|TargetLost))))

(fn TargetLost
  (Transformation|DetachFromActor "KeepWorld" "KeepWorld" "KeepWorld")
  (Widget|SetVisibility (UserInterface|GetUserWidgetObject (Variables|Default|GetWidget)) "Hidden"))

(fn ConstructionScript)
