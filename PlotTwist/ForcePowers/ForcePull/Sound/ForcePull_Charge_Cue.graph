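# Sound cue for the Force Pull charge sound, built from distance layers with their own attenuation and sound class overrides.
# An IsLocal branch picks 2D attenuation for the local player or close attenuation otherwise, mixed with medium and heavily reduced distant layers.
# Each layer randomly picks one of three waves through a modulator, with pitch pinned on the local and close paths.
# A concurrency override limits simultaneous instances, stopping the oldest.
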SoundCue /PlotTwist/ForcePowers/ForcePull/Sound/ForcePull_Charge_Cue
root=Mixer_2
nodes [
 {
  id=Attenuation_0
  children [Modulator_13]
  properties {attenuationSettings=/Game/Sounds/Attenuation/Fort_ForcePowers_ForcePull_Cast_Medium_Att}
 }
 {
  id=Attenuation_1
  children [Modulator_1]
  properties {attenuationSettings=/Game/Sounds/Attenuation/Fort_ForcePowers_ForcePull_Cast_Close_Att}
 }
 {id=Attenuation_2, children [Modulator_0], properties {attenuationSettings=/Game/Sounds/Attenuation/Fort_ForcePowers_ForcePull_Cast_2D_Att}}
 {
  id=Attenuation_3
  children [Modulator_2]
  properties {attenuationSettings=/Game/Sounds/Attenuation/Fort_ForcePowers_ForcePull_Cast_Distant_Att}
 }
 {id=Branch_0, children [Attenuation_2, Attenuation_1, Attenuation_1], properties {boolParameterName=IsLocal}}
 {id=Mixer_2, children [SoundClass_77, SoundClass_76, SoundClass_0], properties {inputVolume [1.5,1,0.05]}}
 {id=Modulator_0, children [Random_0], properties {pitchMax=1, pitchMin=1, volumeMax=1, volumeMin=1}}
 {id=Modulator_1, children [Random_1], properties {pitchMax=1, pitchMin=1, volumeMax=1, volumeMin=1}}
 {id=Modulator_13, children [Random_3], properties {volumeMax=1, volumeMin=1}}
 {id=Modulator_2, children [Random_2], properties {volumeMax=1, volumeMin=1}}
 {id=Random_0, children [WavePlayer_1,WavePlayer_2,WavePlayer_3], properties {weights [1,1,1]}}
 {id=Random_1, children [WavePlayer_4,WavePlayer_5,WavePlayer_9], properties {weights [1,1,1]}}
 {id=Random_2, children [WavePlayer_10, WavePlayer_11, WavePlayer_21], properties {weights [1,1,1]}}
 {id=Random_3, children [WavePlayer_22, WavePlayer_23, WavePlayer_24], properties {weights [1,1,1]}}
 {
  id=SoundClass_0
  children [Attenuation_3]
  properties {soundClassOverride=&Fort_WeaponRanged_Fire_Distant}
 }
 {
  id=SoundClass_76
  children [Attenuation_0]
  properties {soundClassOverride=&Fort_WeaponRanged_Fire_Distant}
 }
 {id=SoundClass_77, children [Branch_0], properties {soundClassOverride=&Fort_WeaponRanged_Fire_3P_PSM}}
 {id=WavePlayer_1, properties {soundWaveAssetPtr=&/ForcePull_Charge_01}}
 {id=WavePlayer_10, properties {soundWaveAssetPtr=&/ForcePull_Charge_DIstant_01}}
 {id=WavePlayer_11, properties {soundWaveAssetPtr=&/ForcePull_Charge_DIstant_02}}
 {id=WavePlayer_2, properties {soundWaveAssetPtr=&/ForcePull_Charge_02}}
 {id=WavePlayer_21, properties {soundWaveAssetPtr=&/ForcePull_Charge_DIstant_03}}
 {id=WavePlayer_22, properties {soundWaveAssetPtr=&/ForcePull_Charge_Medium_01}}
 {id=WavePlayer_23, properties {soundWaveAssetPtr=&/ForcePull_Charge_Medium_02}}
 {id=WavePlayer_24, properties {soundWaveAssetPtr=&/ForcePull_Charge_Medium_03}}
 {id=WavePlayer_3, properties {soundWaveAssetPtr=&/ForcePull_Charge_03}}
 {id=WavePlayer_4, properties {soundWaveAssetPtr=&ForcePull_Charge_01}}
 {id=WavePlayer_5, properties {soundWaveAssetPtr=&ForcePull_Charge_02}}
 {id=WavePlayer_9, properties {soundWaveAssetPtr=&ForcePull_Charge_03}}
]
properties {
 bOverrideConcurrency=true
 maxDistance=2097152
 soundClassObject=&Fort_Gadgets_Items
 volumeMultiplier=0.8
 attenuationOverrides {
  bAttenuate=false
  distanceAlgorithm=NaturalSound
  falloffDistance=3800
  lPFRadiusMax=3000
  lPFRadiusMin=2000
  attenuationShapeExtents {x=2200}
 }
 concurrencyOverrides {maxCount=2, resolutionRule=StopOldest}
}
