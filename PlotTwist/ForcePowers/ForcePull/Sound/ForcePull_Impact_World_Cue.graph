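# Sound cue for the Force Pull power's world impact, playing random impact variations with a fixed delay.
# A quality switch picks from five waves on its first two levels and a three-wave subset on the last two.
# The delayed result gets pitch variation and doubled volume, then a third-person ranged-fire class override despite the cue's Impacts class.
# It borrows Force Push impact attenuation and caps concurrency with its own override.
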SoundCue /PlotTwist/ForcePowers/ForcePull/Sound/ForcePull_Impact_World_Cue
root=Attenuation_0
nodes [
 {
  id=Attenuation_0
  children [SoundClass_0]
  properties {attenuationSettings=/Game/Sounds/Attenuation/Fort_ForcePowers_ForcePush_Blast_Impact_World_Att}
 }
 {id=Delay_0, children [QualityLevel_0], properties {delayMax=0.4, delayMin=0.4}}
 {id=Modulator_0, children [Delay_0], properties {pitchMax=1.1, pitchMin=0.9, volumeMax=2, volumeMin=2}}
 {id=QualityLevel_0, children [Random_1,Random_1,Random_0,Random_0]}
 {id=Random_0, children [WavePlayer_7,WavePlayer_8,WavePlayer_9], properties {bShouldExcludeFromBranchCulling=true, weights [1,1,1]}}
 {id=Random_1, children [WavePlayer_5,WavePlayer_6,WavePlayer_7,WavePlayer_8,WavePlayer_9], properties {weights [1,1,1,1,1]}}
 {id=SoundClass_0, children [Modulator_0], properties {soundClassOverride=&Fort_WeaponRanged_Fire_3P_PSM}}
 {id=WavePlayer_5, properties {soundWaveAssetPtr=&/ForcePull_ImpactWorld_01}}
 {id=WavePlayer_6, properties {soundWaveAssetPtr=&/ForcePull_ImpactWorld_02}}
 {id=WavePlayer_7, properties {soundWaveAssetPtr=&/ForcePull_ImpactWorld_03}}
 {id=WavePlayer_8, properties {soundWaveAssetPtr=&/ForcePull_ImpactWorld_04}}
 {id=WavePlayer_9, properties {soundWaveAssetPtr=&/ForcePull_ImpactWorld_05}}
]
properties {
 bOverrideConcurrency=true
 maxDistance=4100
 soundClassObject=&Fort_Impacts
 volumeMultiplier=1
 attenuationOverrides {falloffDistance=6000, attenuationShapeExtents {x=1000}}
 concurrencyOverrides {maxCount=3}
}
